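# Parent material for the RelayStick backpack, wrapping the 2023 character PreFX/PostFX functions with masked effects for body, inner and outer wings, and emissive areas.
# A TechArt mask drives chained attribute blends adding pastel fresnel sheen, wing gradients from curve atlases and body glow.
# The EnableReactivity switch adds audio-reactive colors, tints, radial electric-noise energy and loudness-boosted glow.
# A DEBUG MASK switch can output the raw mask as emissive.
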Material /BRCosmetics/Accessories/FORT_Backpacks/Backpack_M_MED_RelayStick/Materials/M_RelayStick_Backpack_Parent
graph {
 properties {
  bUseMaterialAttributes=true
  bUsedWithMeshDeformer=true
  bUsedWithSkeletalMesh=true
  pixelDepthOffsetMode=PDOM_Legacy
  previewMesh=/BRCosmetics/Accessories/FORT_Backpacks/Backpack_M_MED_RelayStick/Meshes/M_MED_RelayStick_Pack
  shadingModel=MSM_FromMaterialExpression
  usedShadingModels=Default Lit | Subsurface | Clear Coat | Subsurface Profile | Cloth
 }
 nodes [
  n1 = TextureSampleParameter2D() {
   group=RelayStick
   parameterName=TechArtFX_Mask
   samplerSource=SSM_Wrap_WorldGroupSettings
   samplerType=SAMPLERTYPE_LinearColor
   texture=/BRCosmetics/Accessories/FORT_Backpacks/Backpack_M_MED_RelayStick/Textures/T_RelayStick_Backpack_TechArt
  }
  n2 = MaterialFunctionCall() {materialFunction=&MF_CharacterParent_2023_PreFX}
  MaterialFunctionCall(
   Material=StaticSwitchParameter(True=SetMaterialAttributes(Emissive Color=@n1.RGB) {attributeSetTypes [#1]}, False=NamedRerouteUsage()) {parameterName=DEBUG MASK}
   M=@n2.M
   Pre Skinned Local Position=@n2.M
   Distance Blend=@n2.M
   Fresnel=@n2.M
  ) {materialFunction=&MF_CharacterParent_2023_PostFX}
  n3 = ScalarParameter() {defaultValue=4, parameterName=OuterWingBaseColorBoost}
  n4 = Power(Base=NamedRerouteUsage())
  Multiply(A=@n3, B=@n4)
  n5 = MaterialFunctionCall() {materialFunction=/Game/Packages/Fortress_SharedMaterials/Base_Material_Functions/BaseCharacterShaderFunctions/Audio/MF_AudioReactivity}
  n6 = ScalarParameter() {defaultValue=1, parameterName=EnableAudioReactivity, sliderMax=1}
  NamedRerouteDeclaration(None=Multiply(A=@n5.Amplitude, B=@n6)) {name=Amplitude, nodeColor=<rgba 1 0 0.5156 1>}
  NamedRerouteDeclaration(None=StaticSwitchParameter(True=Constant() {r=1}, False=Constant()) {group=RelayStick, parameterName=EnableReactivity}) {name=StateMask, nodeColor=<rgba 0.2188 0 1 1>}
  n7 = VectorParameter() {group=RelayStick, parameterName=PastelColorA2, defaultValue=<rgba 0.0728 0 0.6562 1>}
  n8 = VectorParameter() {group=RelayStick, parameterName=PastelColorA, defaultValue=<rgba 0.8426 0 1 1>}
  n9 = PreSkinnedPosition()
  n10 = ComponentMask(None=NamedRerouteDeclaration(None=VertexInterpolator(VS=@n9).PS) {name=PreSkinnedPos, nodeColor=<rgba 0.5938 0 1 1>}) {g=true, r=true}
  n11 = Add(
   A=ComponentMask(None=MaterialFunctionCall(Vector or UVs=@n10) {materialFunction=&VectorToRadialValue}.Radial Coordinates) {r=true}
   B=Multiply(A=Time()) {constB=0.06}
  )
  n12 = Fresnel() {baseReflectFraction=0, exponent=0.5}
  n13 = TextureCoordinate()
  n14 = DDX(Value=@n13)
  n15 = DDY(Value=@n13)
  n16 = TextureSampleParameter2D!(UVs=AppendVector(A=@n11, B=@n12), "DDX(UVs)"=@n14, "DDY(UVs)"=@n15, Apply View MipBias) {
   mipValueMode=TMVM_Derivative
   parameterName=RelayStick_Pattern
   samplerSource=SSM_Wrap_WorldGroupSettings
   samplerType=SAMPLERTYPE_LinearColor
   texture=/Game/Accessories/FORT_Backpacks/Backpack_M_MED_RelayStick/Textures/T_RelayStick_Pattern_A
  }
  n17 = VectorParameter() {group=RelayStick, parameterName=PastelColorB2, defaultValue=<rgba 0 0.2729 0.7292 1>}
  n18 = VectorParameter() {group=RelayStick, parameterName=PastelColorB, defaultValue=<rgba 0 1 0.7071 1>}
  n19 = TextureSampleParameter2D(UVs=AppendVector(A=Multiply(A=@n11) {constB=2}, B=@n12), "DDX(UVs)"=@n14, "DDY(UVs)"=@n15) {
   mipValueMode=TMVM_Derivative
   parameterName=RelayStick_Pattern
   samplerSource=SSM_Wrap_WorldGroupSettings
   samplerType=SAMPLERTYPE_LinearColor
   texture=&T_RelayStick_Pattern_A
  }
  n20 = TextureSampleParameter2D(UVs=AppendVector(A=Multiply(A=@n11) {constB=3}, B=@n12), "DDX(UVs)"=@n14, "DDY(UVs)"=@n15) {
   mipValueMode=TMVM_Derivative
   parameterName=RelayStick_Pattern
   samplerSource=SSM_Wrap_WorldGroupSettings
   samplerType=SAMPLERTYPE_LinearColor
   texture=&T_RelayStick_Pattern_A
  }
  n21 = LinearInterpolate(
   A=LinearInterpolate(A=LinearInterpolate(A=@n7.RGB, B=@n8.RGB, Alpha=@n16.R), B=LinearInterpolate(A=@n17.RGB, B=@n18.RGB, Alpha=@n19.G), Alpha=@n16.B)
   B=LinearInterpolate(
    A=VectorParameter() {group=RelayStick, parameterName=PastelColorC2, defaultValue=<rgba 0.2486 0.849 0.0058 1>}.RGB
    B=VectorParameter() {group=RelayStick, parameterName=PastelColorC, defaultValue=<rgba 0.7286 1 0.3333 1>}.RGB
    Alpha=@n20.B
   )
   Alpha=@n20.R
  )
  n22 = ScalarParameter() {defaultValue=2, group=RelayStick, parameterName=PastelColorSheen_Exponent}
  NamedRerouteDeclaration(None=Multiply(A=Power(Base=@n21, Exp=@n22), B=ScalarParameter() {defaultValue=0.3, group=RelayStick, parameterName=PastelColorSheen_Brightness})) {name=PastelSheen, nodeColor=<rgba 0 0.9062 1 1>}
  NamedRerouteDeclaration(None=Multiply(A=@n5.Spectrum, B=@n6)) {name=Spectrum, nodeColor=<rgba 0 0.9062 1 1>}
  n23 = ChannelMaskParameter(None=@n1.RGBA) {maskChannel=Green, parameterName=BodyMaskChannel, defaultValue=<rgba 0 1 0 0>}
  n24 = ChannelMaskParameter(None=@n1.RGBA) {maskChannel=Blue, parameterName=OuterFrameMaskChannel, defaultValue=<rgba 0 0 1 0>}
  n25 = Subtract(A=@n23.R, B=@n24.R) {desc=Emissive}
  NamedRerouteDeclaration(None=Power(Base=Saturate(None=@n25))) {name=EmissiveMask, nodeColor=<rgba 0.8047 1 0 1>}
  NamedRerouteDeclaration(None=Multiply(A=@n23.R, B=@n24.R) {desc=Body}) {name=BodyMask, nodeColor=<rgba 0 0.4609 1 1>}
  NamedRerouteDeclaration(None=ChannelMaskParameter(None=@n1.RGBA) {parameterName=InnerWingsMaskChannel}.R) {name=InnerWingsMask, nodeColor=<rgba 0 0.7891 1 1>}
  n26 = Subtract(A=@n24.R, B=@n23.R)
  NamedRerouteDeclaration(None=Power(Base=Saturate(None=@n26))) {name=OuterWingsMask, nodeColor=<rgba 0 1 0.5078 1>}
  n27 = Multiply(
   A=ScalarParameter() {defaultValue=0.3, parameterName=InnerWingsGradient_NormalBasedOffsetAmt}
   B=ComponentMask(None=GetMaterialAttributes(None=@n2.Result) {attributeGetTypes [#2]}.Normal) {g=true}
  )
  n28 = Power(Base=NamedRerouteUsage(), Exp=ScalarParameter() {defaultValue=6, parameterName=Reactive_Sensitivity})
  n29 = Multiply(A=@n28, B=ScalarParameter() {defaultValue=0.05, parameterName=Reactive_PushAmt})
  n30 = TextureSample() {
   samplerType=SAMPLERTYPE_LinearColor
   texture=/BRCosmetics/Accessories/FORT_Backpacks/Backpack_M_MED_RelayStick/Textures/T_RelayStick_Backpack_InnerWingsGradient
  }
  n31 = Power(Base=@n30.R)
  n32 = Multiply(
   A=NamedRerouteDeclaration(None=OneMinus(None=@n31)) {name=InnerWingsGradient, nodeColor=<rgba 0.4297 0 1 1>}
   B=ScalarParameter() {defaultValue=0.8, parameterName=InnerWings_Gradient_Scale}
  ) {constB=0.75}
  n33 = CurveAtlasRowParameter(CurveTime=Add(A=Subtract(A=@n27, B=@n29), B=@n32)) {
   atlas=/Game/Weapons/FORT_Melee/Pickaxe_RelayStick/Textures/CA_RelayStick
   curve=/Game/Weapons/FORT_Melee/Pickaxe_RelayStick/Textures/C_RelayStick_Backpack_Reactive
   defaultValue=5
   parameterName=InnerWings_Gradient_Reactive
  }
  n34 = Add(A=@n29) {constB=0.8}
  n35 = TextureSample() {
   samplerSource=SSM_Wrap_WorldGroupSettings
   texture=/BRCosmetics/Accessories/FORT_Backpacks/Backpack_M_MED_RelayStick/Textures/T_RelayStick_Backpack_Swirls
  }
  n36 = LinearInterpolate(A=@n35.G, B=@n32) {constB=0.03}
  n37 = CurveAtlasRowParameter(CurveTime=Add(A=LinearInterpolate(A=@n34, Alpha=@n36) {constA=0.8}, B=@n27)) {
   atlas=&CA_RelayStick
   curve=&C_RelayStick_Backpack_Reactive
   defaultValue=5
   parameterName=InnerWings_Gradient_Reactive
  }
  n38 = VectorParameter() {parameterName=InnerWings_Reactive_Energy, defaultValue=<rgba 1 0 0.635 1>}
  n39 = LinearInterpolate(Alpha=@n28) {constA=8}
  n40 = Power(Base=@n30.R) {constExponent=6}
  n41 = Subtract(A=NamedRerouteUsage(), B=Constant3Vector() {constant=<rgba 0 0 -6 0>})
  n42 = Constant3Vector() {constant=<rgba 1 0.1 0.7 0>}
  n43 = MaterialFunctionCall(Vector or UVs=ComponentMask(None=Multiply(A=@n41, B=@n42)) {b=true, r=true}) {materialFunction=&VectorToRadialValue}
  n44 = MaterialFunctionCall(Float2=@n43.Radial Coordinates) {materialFunction=&BreakOutFloat2Components}
  n45 = Multiply(A=@n43.Linear Distance) {constB=2}
  n46 = MaterialFunctionCall(X=@n44.R, Y=Divide(A=Power(Base=@n44.G) {constExponent=1.2}, B=Add(A=@n45))) {materialFunction=&MakeFloat2}
  n47 = Constant2Vector() {g=1, r=7}
  n48 = TextureSample(UVs=Panner(Coordinate=Multiply(A=@n46.Result, B=@n47) {constB=3}) {speedX=0.1, speedY=-0.067}) {
   samplerSource=SSM_Wrap_WorldGroupSettings
   samplerType=SAMPLERTYPE_LinearColor
   texture=/Game/Characters/Player/Male/Medium/Bodies/M_MED_SRDrift/Textures/T_ElectricNoise
  }
  n49 = Constant2Vector() {g=0.783, r=5}
  n50 = TextureSample(UVs=Panner(Coordinate=Multiply(A=@n46.Result, B=@n49)) {speedX=-0.087, speedY=-0.053}) {
   samplerSource=SSM_Wrap_WorldGroupSettings
   samplerType=SAMPLERTYPE_LinearColor
   texture=&T_ElectricNoise
  }
  n51 = Power(Base=Add(A=@n48.R, B=@n50.R)) {constExponent=1}
  n52 = Power(Base=LinearInterpolate(A=Power(Base=@n30.R, Exp=@n39) {constExponent=10}, B=Multiply(A=@n40, B=@n51)) {constAlpha=0.85}) {constExponent=1.1}
  n53 = ScalarParameter() {defaultValue=6, parameterName=InnerWings_Reactive_Energy_Intensity}
  n54 = ScalarParameter() {defaultValue=25, parameterName=InnerWings_Reactive_Energy_Intensity_Loud}
  n55 = Power(Base=ComponentMask(None=NamedRerouteUsage()) {b=true}) {constExponent=4}
  n56 = Add(
   A=LinearInterpolate(A=@n33, B=@n37, Alpha=@n35.G)
   B=Multiply(
    A=MaterialFunctionCall(B=@n38.RGB, Alpha=@n52) {materialFunction=&Lerp_3Color}.Result
    B=LinearInterpolate(A=@n53, B=@n54, Alpha=@n55)
   ) {constB=6}
  )
  n57 = Multiply(A=@n56, B=ScalarParameter() {defaultValue=2, parameterName=Reactive_OverallBoost})
  n58 = ScalarParameter() {defaultValue=1, parameterName=Reactive_OverallBoostCurveExp}
  n59 = ScalarParameter() {defaultValue=1, parameterName=InnerWings_GradientFalloff}
  n60 = WorldPosition() {worldPositionShaderOffset=WPT_CameraRelative}
  n61 = Multiply(
   A=Add(
    A=TransformPosition(None=@n60.XYZ) {transformSourceType=TRANSFORMPOSSOURCE_TranslatedWorld}
    B=VectorParameter() {parameterName=InnerWings_GradientOffset, defaultValue=<rgba 0 3 6 1>}.RGB
   )
   B=VectorParameter() {parameterName=InnerWings_GradientScale, defaultValue=<rgba 1 0.5 0.3 1>}.RGB
  )
  n62 = ScalarParameter() {defaultValue=40, parameterName=InnerWings_GradientRadius}
  n63 = ScalarParameter() {defaultValue=15, parameterName=InnerWings_GradientRadius_Reactive}
  n64 = NamedRerouteUsage()
  n65 = Add(A=Divide(A=Length(None=@n61) {is=class MaterialExpressionLength}, B=LinearInterpolate(A=@n62, B=@n63, Alpha=@n64)))
  n66 = SmoothStep(Value=OneMinus(None=Divide(A=@n59, B=@n65) {constA=3}))
  n67 = ScalarParameter() {defaultValue=1, parameterName=InnerWings_FresnelExponent, sliderMax=1}
  n68 = LinearInterpolate(
   A=@n66
   B=Multiply(A=@n66, B=Fresnel(ExponentIn=@n67) {baseReflectFraction=0, exponent=2})
   Alpha=ScalarParameter() {defaultValue=0.7, parameterName=InnerWings_FresnelAmount, sliderMax=1}
  ) {constAlpha=0.8}
  n69 = ScalarParameter() {defaultValue=1, group=06_Thin Film, parameterName=InnerWings_ThinFilmWarp, sortPriority=4}
  n70 = Add(
   A=Multiply(
    A=ScalarParameter() {defaultValue=1, group=06_Thin Film, parameterName=InnerWings_ThinFilmScale, sortPriority=5}
    B=Power(Base=@n68, Exp=@n69)
   )
   B=ScalarParameter() {group=06_Thin Film, parameterName=InnerWings_ThinFilmOffset}
  )
  n71 = Multiply(A=@n35.B, B=ScalarParameter() {defaultValue=0.1, parameterName=InnerWings_SwirlsAmount}) {constB=0.05}
  n72 = CurveAtlasRowParameter(CurveTime=Frac(None=Add(A=@n70, B=@n71))) {
   atlas=&CA_RelayStick
   curve=/Game/Weapons/FORT_Melee/Pickaxe_RelayStick/Textures/C_RelayStick_Backpack
   defaultValue=2
   parameterName=InnerWings_Gradient
  }
  n73 = StaticSwitchParameter(True=LinearInterpolate(A=@n56, B=Power(Base=@n57, Exp=@n58), Alpha=@n28), False=@n72) {group=RelayStick, parameterName=EnableReactivity}
  n74 = ScalarParameter() {defaultValue=5, group=06_Thin Film, parameterName=InnerWings_ThinFilmIntensity, sortPriority=2}
  NamedRerouteDeclaration(None=MaterialFunctionCall!(DayBloom_Mult=ScalarParameter() {defaultValue=1, parameterName=InnerWings_DayEmissiveScale}, In=Multiply(A=@n73, B=@n74)) {materialFunction=/Game/Packages/Fortnite_UniqueMaterials_1/MH_Materials/DayNight_Auto_Bloom_Control}.Result) {name=InnerWingsEmissive, nodeColor=<rgba 0.3125 1 0 1>}
  n75 = GetMaterialAttributes(None=@n2.Result) {attributeGetTypes [#3, #1]}
  n76 = Multiply(A=@n75.Base Color, B=LinearInterpolate(A=@n3, Alpha=@n4) {constA=1}) {constB=4}
  n77 = NamedRerouteUsage()
  n78 = BlendMaterialAttributes(
   A=@n2.Result
   B=SetMaterialAttributes!(
    MaterialAttributes=@n75.MaterialAttributes
    Base Color=StaticSwitchParameter(True=@n76, False=@n75.Base Color) {group=RelayStick, parameterName=EnableReactivity}
    Emissive Color=Add(A=@n75.Emissive Color, B=@n77)
   ) {attributeSetTypes [#3, #1]}
   Alpha=NamedRerouteUsage()
  )
  n79 = GetMaterialAttributes(None=@n78) {attributeGetTypes [#3, #1]}
  n80 = Multiply(A=@n79.Base Color, B=VectorParameter() {parameterName=BodyTint, defaultValue=<rgba 0 0.1335 1 1>}.RGB)
  n81 = Multiply(
   A=Add(A=@n79.Emissive Color, B=NamedRerouteUsage())
   B=VectorParameter() {parameterName=BodyEmissiveTint, defaultValue=<rgba 0 0.1335 1 1>}.RGB
  )
  n82 = StaticSwitchParameter(
   True=BlendMaterialAttributes(
    A=@n78
    B=SetMaterialAttributes(MaterialAttributes=@n79.MaterialAttributes, Base Color=@n80, Emissive Color=@n81) {attributeSetTypes [#3, #1]}
    Alpha=NamedRerouteUsage()
   )
   False=@n78
  ) {group=RelayStick, parameterName=EnableReactivity}
  n83 = GetMaterialAttributes(None=@n82) {attributeGetTypes [#3, #1]}
  n84 = Multiply(A=@n83.Base Color, B=ScalarParameter() {parameterName=InnerWings_Reactive_BCDarkening})
  n85 = NamedRerouteUsage()
  n86 = BlendMaterialAttributes(
   A=@n82
   B=SetMaterialAttributes(
    MaterialAttributes=@n83.MaterialAttributes
    Base Color=StaticSwitchParameter(True=@n84, False=@n83.Base Color) {group=RelayStick, parameterName=EnableReactivity}
    Emissive Color=Add(A=@n83.Emissive Color, B=@n85)
   ) {attributeSetTypes [#3, #1]}
   Alpha=NamedRerouteUsage()
  )
  n87 = GetMaterialAttributes(None=@n86) {attributeGetTypes [#3, #1]}
  n88 = Fresnel(ExponentIn=ScalarParameter() {defaultValue=1, group=RelayStick, parameterName=BodyGlowFresnelExp}) {baseReflectFraction=0, exponent=3}
  n89 = Multiply(A=Subtract(A=@n88, B=ScalarParameter() {defaultValue=0.1, group=RelayStick, parameterName=BodyGlow_HighlightTreshold, sliderMax=1})) {constB=-5}
  n90 = Multiply(
   A=LinearInterpolate(
    A=VectorParameter() {group=RelayStick, parameterName=BodyGlowColor, defaultValue=<rgba 0 1 0.7355 1>}.RGB
    B=VectorParameter() {group=RelayStick, parameterName=BodyGlowColorEdges, defaultValue=<rgba 0 1 0.4914 1>}.RGB
    Alpha=@n88
   )
   B=LinearInterpolate(
    A=ScalarParameter() {defaultValue=0.1, group=RelayStick, parameterName=BodyGlowBaseIntensity}
    B=ScalarParameter() {defaultValue=5, group=RelayStick, parameterName=BodyGlowHighlightIntensity}
    Alpha=Saturate(None=@n89)
   )
  )
  n91 = NamedRerouteUsage()
  n92 = LinearInterpolate(
   A=ScalarParameter() {defaultValue=1, group=RelayStick, parameterName=BodyGlowStrength}
   B=ScalarParameter() {defaultValue=4, group=RelayStick, parameterName=BodyGlowStrength_Loud}
   Alpha=Power(
    Base=ComponentMask(None=@n91) {g=true}
    Exp=ScalarParameter() {defaultValue=6, group=RelayStick, parameterName=BodyGlowStrength_AudioSensitivity}
   )
  ) {constA=1}
  n93 = SetMaterialAttributes(
   MaterialAttributes=@n87.MaterialAttributes
   Base Color=Multiply(A=@n87.Base Color, B=VectorParameter() {group=RelayStick, parameterName=BodyGlowBCTint, defaultValue=<rgba 0 1 0.4914 1>}.RGB)
   Emissive Color=Add(A=@n87.Emissive Color, B=Multiply(A=@n90, B=@n92))
  ) {attributeSetTypes [#3, #1]}
  n94 = NamedRerouteUsage()
  NamedRerouteDeclaration(None=StaticSwitchParameter(True=BlendMaterialAttributes(A=@n86, B=@n93, Alpha=@n94), False=@n86) {group=RelayStick, parameterName=EnableReactivity}) {name=Blended Material Result, nodeColor=<rgba 0.3125 1 0 1>}
  StaticSwitchParameter() {group=RelayStick, parameterName=EnableReactivity}
 ]
}
