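# Control Rig for the RelayStick backpack that procedurally flaps its two three-segment wings.
# Moving, idle, dive and floating flags set a random speed range, which combines with intermittent random bursts to drive a flip-flopping time loop.
# Per-segment functions map that loop through curves to Z rotations, negated for the right side, with time offsets on the second and third segments for follow-through.
# A final step slightly offsets the pack bone.
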ControlRigBlueprint /BRCosmetics/Accessories/FORT_Backpacks/Backpack_M_MED_RelayStick/Meshes/M_MED_RelayStick_Pack_CtrlRig
graphs [
 New Function_ContainedGraph {nodes [
   n1 = RigVMFunctionEntryNode()
   n2 = &RigVMFunction_AnimEvalRichCurve(
    Curve=(EditorCurveData=(Keys=((InterpMode=RCIM_Cubic,TangentMode=RCTM_User),(InterpMode=RCIM_Cubic,TangentMode=RCTM_User,Time=1,Value=50))))
    SourceMinimum=0
    TargetMinimum=0
    Value=@n1.Flip Flop
   )
   n3 = &RigVMFunction_MathDoubleMul(B=0.7, A=@n2.Result)
   RigVMFunction_MathQuaternionFromEuler = &RigVMFunction_MathQuaternionFromEuler(Euler=(X=0,Y=0,Z=0), Euler.Z=@n3.Result)
   Multiply = &RigVMFunction_MathDoubleMul(B=-1, A=@n3.Result)
   n4 = &RigVMFunction_MathDoubleMul(B=0.2, A=@n2.Result)
   > &RigUnit_ModifyTransforms(
    ItemToModify="((Item=(Type=Bone,Name=\"wing_01_l\"),Transform=(Rotation=(X=0,Y=0,Z=0,W=1),Translation=(X=0,Y=0,Z=0),Scale3D=(X=1,Y=1,Z=1))),(Item=(Type=Bone,Name=\"wing_01_r\"),Transform=(Rotation=(X=0,Y=0,Z=0,W=1),Translation=(X=0,Y=0,Z=0),Scale3D=(X=1,Y=1,Z=1))),(Item=(Ty ...(truncated)"
    Weight=1
    WeightMinimum=0
    ItemToModify.0.Transform.Rotation=@RigVMFunction_MathQuaternionFromEuler.Result
    ItemToModify.1.Transform.Rotation=&RigVMFunction_MathQuaternionFromEuler(Euler=(X=0,Y=0,Z=0), Euler.Z=@Multiply.Result).Result
    ItemToModify.2.Transform.Rotation=&RigVMFunction_MathQuaternionFromEuler(Euler=(X=0,Y=0,Z=0), Euler.Z=@n4.Result).Result
    ItemToModify.3.Transform.Rotation=&RigVMFunction_MathQuaternionFromEuler(Euler=(X=0,Y=0,Z=0), Euler.Z=&RigVMFunction_MathDoubleMul(B=-1, A=@n4.Result).Result).Result
   )
   > RigVMFunctionReturnNode()
  ]}
 New Function_ContainedGraph {nodes [
   n1 = RigVMFunctionEntryNode()
   n2 = &RigVMFunction_AnimEvalRichCurve(
    Curve=(EditorCurveData=(Keys=((InterpMode=RCIM_Cubic,TangentMode=RCTM_User,Value=-5),(InterpMode=RCIM_Cubic,TangentMode=RCTM_User,Time=1,Value=12))))
    SourceMinimum=0
    TargetMinimum=0
    Value=@n1.Flip Flop
   )
   TimeOffset = /Script/RigVM.RigVMFunction_TimeOffsetFloat(BufferSize=50, TimeRange=2, Value=@n2.Result, SecondsAgo=@n1.VOT Seconds) {is=struct RigVMUnitNode, template="TimeOffset::Execute(in Value,in SecondsAgo,in BufferSize,in TimeRange,out Result)"}
   > &RigUnit_ModifyTransforms(
    ItemToModify="((Item=(Type=Bone,Name=\"wing_02_l\"),Transform=(Rotation=(X=0,Y=0,Z=0,W=1),Translation=(X=0,Y=0,Z=0),Scale3D=(X=1,Y=1,Z=1))),(Item=(Type=Bone,Name=\"wing_02_r\"),Transform=(Rotation=(X=0,Y=0,Z=0,W=1),Translation=(X=0,Y=0,Z=0),Scale3D=(X=1,Y=1,Z=1))),(Item=(Ty ...(truncated)"
    Weight=1
    WeightMinimum=0
    ItemToModify.1.Transform.Rotation=&RigVMFunction_MathQuaternionFromEuler(Euler=(X=0,Y=0,Z=0), Euler.Z=&RigVMFunction_MathDoubleMul(B=-1, A=@TimeOffset.Result).Result).Result
    ItemToModify.0.Transform.Rotation=&RigVMFunction_MathQuaternionFromEuler(Euler=(X=0,Y=0,Z=0), Euler.Z=@TimeOffset.Result).Result
   )
   > RigVMFunctionReturnNode()
  ]}
 New Function_ContainedGraph {nodes [
   n1 = RigVMFunctionEntryNode()
   n2 = &RigVMFunction_AnimEvalRichCurve(
    Curve=(EditorCurveData=(Keys=((InterpMode=RCIM_Cubic,TangentMode=RCTM_User,Value=-5),(InterpMode=RCIM_Cubic,TangentMode=RCTM_User,Time=1,Value=10))))
    SourceMinimum=0
    TargetMinimum=0
    Value=@n1.Flip Flop
   )
   n3 = &RigVMFunction_TimeOffsetFloat(BufferSize=50, TimeRange=2, Value=@n2.Result, SecondsAgo=@n1.VOT Seconds) {template="TimeOffset::Execute(in Value,in SecondsAgo,in BufferSize,in TimeRange,out Result)"}
   > &RigUnit_ModifyTransforms(
    ItemToModify="((Item=(Type=Bone,Name=\"wing_03_l\"),Transform=(Rotation=(X=0,Y=0,Z=0,W=1),Translation=(X=0,Y=0,Z=0),Scale3D=(X=1,Y=1,Z=1))),(Item=(Type=Bone,Name=\"wing_03_r\"),Transform=(Rotation=(X=0,Y=0,Z=0,W=1),Translation=(X=0,Y=0,Z=0),Scale3D=(X=1,Y=1,Z=1))),(Item=(Ty ...(truncated)"
    Weight=1
    WeightMinimum=0
    ItemToModify.1.Transform.Rotation=&RigVMFunction_MathQuaternionFromEuler(Euler=(X=0,Y=0,Z=0), Euler.Z=&RigVMFunction_MathDoubleMul(B=-1, A=@n3.Result).Result).Result
    ItemToModify.0.Transform.Rotation=&RigVMFunction_MathQuaternionFromEuler(Euler=(X=0,Y=0,Z=0), Euler.Z=@n3.Result).Result
   )
   > RigVMFunctionReturnNode()
  ]}
 RigVMFunctionLibrary {nodes [
   Wing 01 Logic = RigVMCollapseNode(ExecuteContext=())
   Wing 02 Logic = RigVMCollapseNode(ExecuteContext=(), VOT Seconds=0.1)
   Wing 03 Logic = RigVMCollapseNode(ExecuteContext=())
  ]}
 RigVMFunction_Sequence_1_ContainedGraph {nodes [
   n1 = RigVMFunctionEntryNode()
   RigVMFunction_Sequence = &RigVMFunction_Sequence(ExecuteContext=@n1.ExecuteContext)
   n2 = &RigVMFunction_Sequence(ExecuteContext=@RigVMFunction_Sequence.B)
   n3 = &RigVMFunction_Sequence(ExecuteContext=@n2.B)
   RigVMFunctionReturnNode(A=@RigVMFunction_Sequence.A, B=@n2.A, C=@n3.A, D=@n3.B)
  ]}
 RigVMModel {nodes [
   RigUnit_BeginExecution = &RigUnit_BeginExecution()
   RigVMFunction_ControlFlowBranch = &RigVMFunction_ControlFlowBranch(ExecuteContext=@RigUnit_BeginExecution.ExecutePin, Condition=RigVMVariableNode() {variable=IsMoving}.Value)
   n1 = &RigVMFunction_ControlFlowBranch(ExecuteContext=@RigVMFunction_ControlFlowBranch.Completed, Condition=RigVMVariableNode() {variable=IsIDLE}.Value)
   n2 = &RigVMFunction_ControlFlowBranch(ExecuteContext=@n1.Completed, Condition=RigVMVariableNode() {variable=IsDive}.Value)
   n3 = &RigVMFunction_ControlFlowBranch(ExecuteContext=@n2.Completed, Condition=RigVMVariableNode() {variable=IsFloating}.Value)
   RigVMFunction_Sequence = RigVMAggregateNode(ExecuteContext=@n3.Completed)
   n4 = /Script/RigVM.RigVMFunction_RandomFloat(Duration=20, Maximum=2, Minimum=0.8, Seed=500) {is=struct RigVMUnitNode, template="Random::Execute(in Seed,in Minimum,in Maximum,in Duration,out Result)"}
   n5 = &RigVMFunction_RandomFloat(Duration=20, Maximum=20, Minimum=10, Seed=200) {template="Random::Execute(in Seed,in Minimum,in Maximum,in Duration,out Result)"}
   n6 = &RigVMFunction_MathDoubleAdd(B=&RigVMFunction_MathDoubleMul(B=2, A=@n4.Result).Result, A=@n5.Result)
   GreaterEqual = /Script/RigVM.RigVMFunction_MathDoubleGreaterEqual(
    A=&RigVMFunction_MathDoubleAdd(B=0.1, A=/Script/RigVM.RigVMFunction_TimeLoop(Normalize=false, Speed=1, Duration=@n6.Result) {is=struct RigVMUnitNode}.Relative).Result
    B=@n5.Result
   ) {is=struct RigVMUnitNode, template="GreaterEqual::Execute(in A,in B,out Result)"}
   DISPATCH_RigVMDispatch_If = RigVMDispatchNode(
    False=0
    Condition=@GreaterEqual.Result
    True=&RigVMFunction_MathDoubleMul(
     A=@n4.Result
     B=&RigVMFunction_RandomFloat(Duration=10, Maximum=8, Minimum=6, Seed=500) {template="Random::Execute(in Seed,in Minimum,in Maximum,in Duration,out Result)"}.Result
    ).Result
   ) {template="DISPATCH_RigVMDispatch_If(in Condition,in True,in False,out Result)"}
   n7 = &RigVMFunction_TimeLoop(Normalize=false, Speed=10, Duration=&RigVMFunction_MathDoubleAdd(B=2, A=@DISPATCH_RigVMDispatch_If.Result).Result)
   n8 = RigVMVariableNode() {variable=State Change Random}
   Random = &RigVMFunction_RandomFloat(Duration=2, Seed=500, Minimum=@n8.Value.X, Maximum=@n8.Value.Y) {template="Random::Execute(in Seed,in Minimum,in Maximum,in Duration,out Result)"}
   Add = &RigVMFunction_MathDoubleAdd(
    B=RigVMDispatchNode(
     True=@DISPATCH_RigVMDispatch_If.Result
     Condition=&RigVMFunction_MathDoubleGreaterEqual(A=@DISPATCH_RigVMDispatch_If.Result, B=@n7.Relative) {template="GreaterEqual::Execute(in A,in B,out Result)"}.Result
     False=@n7.Relative
    ) {template="DISPATCH_RigVMDispatch_If(in Condition,in True,in False,out Result)"}.Result
    A=@Random.Result
   )
   RigVMFunction_TimeLoop = &RigVMFunction_TimeLoop(Duration=1, Normalize=true, Speed=@Add.Result)
   Wing 01 Logic = RigVMFunctionReferenceNode(ExecuteContext=@RigVMFunction_Sequence.A, Flip Flop=@RigVMFunction_TimeLoop.FlipFlop)
   Wing 02 Logic = RigVMFunctionReferenceNode(VOT Seconds=0.2, ExecuteContext=@RigVMFunction_Sequence.B, Flip Flop=@RigVMFunction_TimeLoop.FlipFlop)
   Wing 03 Logic = RigVMFunctionReferenceNode(VOT Seconds=0.2, ExecuteContext=@RigVMFunction_Sequence.C, Flip Flop=@RigVMFunction_TimeLoop.FlipFlop)
   &RigUnit_ModifyTransforms(
    ItemToModify=((Item=(Type=Bone,Name="pack"),Transform=(Rotation=(X=0,Y=0,Z=0,W=1),Translation=(X=0,Y=-1.5,Z=0),Scale3D=(X=1,Y=1,Z=1))))
    Weight=1
    WeightMinimum=0
    ExecutePin=@RigVMFunction_Sequence.D
   )
   RigVMVariableNode(Value=(X=3,Y=4,Z=0), ExecuteContext=@n3.True) {variable=State Change Random}
   RigVMVariableNode(Value=(X=6,Y=7,Z=0), ExecuteContext=@n2.True) {variable=State Change Random}
   RigVMVariableNode(Value=(X=0.5,Y=1,Z=0), ExecuteContext=@n1.True) {variable=State Change Random}
   RigVMVariableNode(Value=(X=3,Y=3.5,Z=0), ExecuteContext=@RigVMFunction_ControlFlowBranch.True) {variable=State Change Random}
  ]}
 RigVMModel setup {nodes [&RigUnit_PrepareForExecution()]}
]
hierarchy [
 root {loc=<0 0 0>, rot=<0 0 0 1>}
 pack {parent=root, loc=<0 -0.7316 0>, rot=<0 0 0 1>}
 wingmetal_l {parent=pack, loc=<-1.3806 -0.3708 -7.251>, rot=<0 0 0.9045 -0.4266>}
 base_wing_l {parent=wingmetal_l, loc=<1.0481 -0.562 -0.3182>, rot=<0 0 0.2932 -0.956>}
 wing_01_l {parent=base_wing_l, loc=<0 0 0>, rot=<0 0 0 1>}
 wing_02_l {parent=wing_01_l, loc=<3.9801 0 0>, rot=<0 0 0 1>}
 wing_03_l {parent=wing_02_l, loc=<3.9801 0 0>, rot=<0 0 0 1>}
 wingmetal_r {parent=pack, loc=<1.3806 -0.3708 -7.251>, rot=<0 0 0.3868 -0.9222>}
 base_wing_r {parent=wingmetal_r, loc=<1.093 0.4687 -0.3182>, rot=<0 0 -0.2513 -0.9679>}
 wing_01_r {parent=base_wing_r, loc=<0 0 0>, rot=<0 0 0 1>}
 wing_02_r {parent=wing_01_r, loc=<3.9801 0 0>, rot=<0 0 0 1>}
 wing_03_r {parent=wing_02_r, loc=<3.9802 0 0>, rot=<0 0 0 -1>}
 base_antennae_l {parent=pack, loc=<-0.5689 -0.6452 0.1242>, rot=<0 0.7071 0 -0.7071>}
 dyn_antennae_l {parent=base_antennae_l, loc=<0 0 0>, rot=<0 0 0 1>}
 base_antennae_r {parent=pack, loc=<0.5689 -0.6452 0.1242>, rot=<0 0.7071 0 -0.7071>}
 dyn_antennae_r {parent=base_antennae_r, loc=<0 0 0>, rot=<0 0 0 1>}
 base_bodylower_c {parent=pack, loc=<0 -2.1212 -12.3511>, rot=<-0.4097 -0.4097 0.5763 -0.5763>}
 dyn_bodylower_c {parent=base_bodylower_c, loc=<0 0 0>, rot=<0 0 0 1>}
]
variables [
 IsDive {type=bool, default=False}
 IsFloating {type=bool, default=False}
 IsIDLE {type=bool, default=False}
 IsMoving {type=bool, default=False}
 MyPose {type=FRigPose, default=(Elements=,HierarchyTopologyVersion=-1,PoseHash=-1)}
 State Change Random {type=FVector, default=(X=1.5,Y=2.5,Z=0)}
]
